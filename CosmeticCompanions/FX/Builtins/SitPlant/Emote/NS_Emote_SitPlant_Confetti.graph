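# Niagara one-shot confetti effect for the SitPlant companion emote.
# Confetti_Medium fires four grouped instant bursts of palette-colored sprites outward and upward, then applies spin, drag, random gravity and source movement until they shrink or slow and die.
# Spawn group drives offset, speed and visibility tag across two sprite renderers, with a front-end color shift and time-of-day tint on secondary sprite color.
# SparklesLWE is also enabled; Airglow and Smoke are disabled.
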NiagaraSystem /CosmeticCompanions/FX/Builtins/SitPlant/Emote/NS_Emote_SitPlant_Confetti
dependencies {
 usedDataInterfaces [&NiagaraDataInterfaceArrayColor, &NiagaraDataInterfaceArrayFloat, &NiagaraDataInterfaceCurve]
 usedDynamicInputs [&SimulationPosition]
 usedModules [&SystemState]
}
emitters {
 AirglowLWE {bEnabled=false, rendererClasses [], renderers []}
 Confetti_Medium {
  emitterUpdateScript [
   &EmitterState()
   &SpawnBurst_Instantaneous(Spawn Count=100, Use Spawn Probability=false)
   &SpawnBurst_Instantaneous(Spawn Count=15, Spawn Group=1, Spawn Time=0.02, Use Spawn Probability=false) {module=SpawnBurst_Instantaneous001}
   &SpawnBurst_Instantaneous(Spawn Count=5, Spawn Group=2, Use Spawn Probability=false) {module=SpawnBurst_Instantaneous002}
   &SpawnBurst_Instantaneous(Spawn Count=5, Spawn Group=3, Spawn Time=0.02, Use Spawn Probability=false) {module=SpawnBurst_Instantaneous003}
  ]
  particleSpawnScript [
   /SetVariables(Particles.SpawnGroup=$Emitter.SpawnGroup) {isSetParameters=true}
   /SetVariables(Particles.VisibilityTag=&MakeCustomIntFromBool(
     Bool=&SetBoolByIntComparison(A=$Particles.SpawnGroup, B=1, Comparison Type=::NewEnumerator7)
     True INT=0
     False INT=2
    )) {isSetParameters=true}
   &InitializeParticle(
    Color=&SelectLinearColorFromArray(
     Array Sampling Mode=::NewEnumerator0
     Direct Array Index=?
     Color Selection Array=DataInterface() {propertyValues {ColorData [<rgba 1 0.1614 0.9113 1>, <rgba 0.148 0 1 1>, <rgba 0.0772 0.9637 1 1>, <rgba 1 0.9088 0.0365 1>, <rgba 0.9385 0 0 0>]}}
     Random Seed=0
    )
    Lifetime=&RandomRangeFloat(Minimum=1.5, Maximum=3)
    Sprite Size Max=<4 6>
    Sprite Size Min=<3 4>
    Sprite Size Mode=::NewEnumerator2
    UsePositionOffset=false
    Write Material Random=true
   ) {module=InitializeParticle001}
   /SetVariables(
    Particles.Sprite_01_Color:LinearColor=&HueShiftLinearColor(
     Hue=0
     Saturation=&RandomRangeFloat(Minimum=2, Maximum=8)
     Value=&RandomRangeFloat(Minimum=1, Maximum=10)
     Linear Color=&RandomRangeLinearColor(Minimum=<rgba 0.4115 0.4115 0.4115 1>, Maximum=<rgba 1 1 1 1>, Color Channel Mode=::NewEnumerator2)
    )
    Particles.Sprite_01_Size:Vector2f=&MakeVector2D(
     X=&MakeFloatFromVector2d(Vector 2D=&Vector2DFromFloat(Value=&RandomRangeFloat(Minimum=3, Maximum=4)))
     Y=&MakeFloatFromVector2d(
      Vector 2D=&Vector2DFromFloat(Value=&RandomRangeFloat(Minimum=3, Maximum=6))
      Channel=::NewEnumerator1
     )
    )
   ) {isSetParameters=true}
   /SetVariables(
    Particles.Sprite_02_Color:LinearColor=&HueShiftLinearColor(
     Hue=0
     Saturation=&RandomRangeFloat(Minimum=2, Maximum=8)
     Value=&RandomRangeFloat(Minimum=1, Maximum=10)
     Linear Color=&RandomRangeLinearColor(Minimum=<rgba 0.4115 0.4115 0.4115 1>, Maximum=<rgba 1 1 1 1>, Color Channel Mode=::NewEnumerator2)
    )
    Particles.Sprite_02_Size:Vector2f=&MakeVector2D(
     X=&MakeFloatFromVector2d(Vector 2D=&Vector2DFromFloat(Value=&RandomRangeFloat(Minimum=10, Maximum=10)))
     Y=&MakeFloatFromVector2d(
      Vector 2D=&Vector2DFromFloat(Value=&RandomRangeFloat(Minimum=2, Maximum=3))
      Channel=::NewEnumerator1
     )
    )
   ) {isSetParameters=true}
   &ShapeLocation(
    MaskSpawnGroup=false
    Offset=&MakeCustomVectorFromBool(
     Bool=&SetBoolByIntComparison(
      A=1
      B=&MakeCustomIntFromBool(
       Bool=&SetBoolByIntComparison(A=$Particles.SpawnGroup, B=3, Comparison Type=::NewEnumerator4)
       True INT=1
       False INT=&SpawnGroup()
      )
      Comparison Type=::NewEnumerator4
     )
     True Vector=<-25 0 5>
     False Vector=<0 0 -5>
    )
    Offset Coordinate Space=::Local
    Offset Mode=::NewEnumerator0
    Override Seed=false
    Rotation Mode=::NewEnumerator1
    Sphere Radius=10
    Sphere Surface Distribution=0.7238
    UseOwnerScale=true
   )
   &AddVelocity(
    Module.Use Falloff Distance=false
    Origin Offset=<0 0 -1>
    Override Seed=false
    Velocity Mode=::NewEnumerator1
    Velocity Speed=/Niagara/DynamicInputs/Math/LerpMultipleFloats(
     Float Count:ENiagara_MultipleLerpCount=::NewEnumerator0
     1:NiagaraFloat=&RandomRangeFloat(Minimum=50, Maximum=112.5)
     2:NiagaraFloat=&RandomRangeFloat(Minimum=62.5, Maximum=162.5)
     3:NiagaraFloat=&RandomRangeFloat(Minimum=50, Maximum=150)
     4:NiagaraFloat=?
     5:NiagaraFloat=?
     Alpha:NiagaraFloat=&Divide_Float(
      A=&MakeFloatFromInt(Int=3)
      B=&MakeFloatFromInt(Int=$Particles.SpawnGroup)
     )
    )
   ) {module=AddVelocity001}
   /SetVariables(Particles.Color=&MakeLinearColorFromVectorAndFloat(
     "Vector (RGB)"=&AddVectorToPosition(
      Position=HlslExpression() {hlslExpression=Particles.Color.xyz}
      Vector=&FN_FrontEndVector3Modifier(Gameplay=<0 0 0>, Front End=<0 0.1 0.25>)
     )
     "Float (Alpha)"=1
    )) {isSetParameters=true}
   &AddVelocity(
    Cone Angle=&MakeCustomFloatFromBool(
     Bool=&SetBoolByIntComparison(
      A=1
      B=&MakeCustomIntFromBool(
       Bool=&SetBoolByIntComparison(A=$Particles.SpawnGroup, B=3, Comparison Type=::NewEnumerator4)
       True INT=1
       False INT=&SpawnGroup()
      )
      Comparison Type=::NewEnumerator4
     )
     True Float=&RandomRangeFloat(Minimum=170, Maximum=180)
     False Float=&RandomRangeFloat(Minimum=90, Maximum=130)
    )
    Cone Axis=&MakeCustomVectorFromBool(
     Bool=&SetBoolByIntComparison(
      A=1
      B=&MakeCustomIntFromBool(
       Bool=&SetBoolByIntComparison(A=$Particles.SpawnGroup, B=3, Comparison Type=::NewEnumerator4)
       True INT=1
       False INT=&SpawnGroup()
      )
      Comparison Type=::NewEnumerator4
     )
     True Vector=<-1 0 0>
     False Vector=<1 0 0>
    )
    Distribution Along Cone Axis=0.1
    Module.UseConeFalloff=false
    Module.UseVelDistribution=true
    Override Seed=false
    Velocity Mode=::NewEnumerator2
    Velocity Speed=&MakeCustomFloatFromBool(
     Bool=&SetBoolByIntComparison(
      A=1
      B=&MakeCustomIntFromBool(
       Bool=&SetBoolByIntComparison(A=$Particles.SpawnGroup, B=3, Comparison Type=::NewEnumerator4)
       True INT=1
       False INT=&SpawnGroup()
      )
      Comparison Type=::NewEnumerator4
     )
     True Float=&RandomRangeFloat(Minimum=50, Maximum=300)
     False Float=&RandomRangeFloat(Minimum=75, Maximum=700)
    )
   ) {enabled=false}
   &AddVelocity(
    Cone Axis=<0 0 1>
    Module.UseConeFalloff=false
    Module.UseVelDistribution=false
    Override Seed=false
    Velocity Mode=::NewEnumerator2
    Velocity Speed=&RandomRangeFloat(Minimum=300, Maximum=850)
   ) {module=AddVelocity002}
   &CurlNoiseForce(Bias Noise=false, Noise Frequency=10, Noise Strength=100, Pan Noise=false, Pan Noise Field=<0.2016 0 0.5008>) {enabled=false, module=CurlNoiseForce001}
   &ApplyInitialForces()
  ]
  particleUpdateScript [
   &ParticleState()
   &SpriteRotationRate(Rotation Rate=&Multiply_Float(
     A=&SelectFloatFromArray(
      Array Sampling Mode=::NewEnumerator0
      Direct Array Index=?
      Float Selection Array=DataInterface() {propertyValues {FloatData [-300,300]}}
      Random Seed=0
     )
     B=&RandomRangeFloat(Minimum=0.5, Maximum=1)
    ))
   &Drag(
    Drag=&FloatFromCurve(
     FloatCurve=DataInterface() {propertyValues {Curve {keys [{arriveTangent=2.2311, interpMode=RCIM_Cubic, leaveTangent=2.2311, tangentMode=RCTM_User, time=0, value=0.75}, <at 0.25 1 RCIM_Cubic>]}}}
     Scale Curve=&RandomRangeFloat(Minimum=5, Maximum=7)
    )
    Use Linear Drag=true
    Use Rotational Drag=false
   )
   &GravityForce(Gravity=&RandomRangeVector(Minimum=<-100 -100 -350>, Maximum=<100 100 -225>))
   &CurlNoiseForce(Noise Strength=&RandomRangeFloat(Minimum=200, Maximum=300), Pan Noise=false) {enabled=false}
   &InheritSourceMovement(Applied Position Scale=<0.8 0.8 1>, ApplyForce=false, ApplyPosition=true, ApplyPositionFalloff=false)
   &SolveForcesAndVelocity(Clamp Velocity=false, Limit Acceleration=false)
   &ScaleSpriteSize(Scale Factor=&Vector2DFromFloat(Value=&FloatFromCurve(
      FloatCurve=DataInterface() {propertyValues {Curve {keys [
          {arriveTangent=-0.4596, interpMode=RCIM_Cubic, leaveTangent=-0.4596, tangentMode=RCTM_User, time=0.55, value=1}
          {arriveTangent=-4.9031, interpMode=RCIM_Cubic, leaveTangent=-4.9031, tangentMode=RCTM_User, time=1, value=0}
         ]}}}
      Scale Curve=&RandomRangeFloat(Minimum=0.4, Maximum=1, Randomness Mode=::NewEnumerator2, Random Seed=?)
     )))
   &ScaleSpriteSizeBySpeed(Max Scale Factor=<1 1>, Min Scale Factor=<0 0>, Velocity Threshold=20)
   /SetVariables(
    Particles.Sprite_01_Color:LinearColor=&MultiplyLinearColor(Linear Color=$Particles.Initial.Sprite_01_Color, Scale Factor=<1 1 1 1>)
    Particles.Sprite_01_Size:Vector2f=&Multiply_Vector2D(
     A=$Particles.Initial.Sprite_01_Size
     B=&Vector2DFromFloat(Value=&FloatFromCurve(
       FloatCurve=DataInterface() {propertyValues {Curve {keys [
           {arriveTangent=-0.4596, interpMode=RCIM_Cubic, leaveTangent=-0.4596, tangentMode=RCTM_User, time=0.55, value=1}
           {arriveTangent=-4.9031, interpMode=RCIM_Cubic, leaveTangent=-4.9031, tangentMode=RCTM_User, time=1, value=0}
          ]}}}
       CurveIndex=&Divide_Float(
        A=$Particles.NormalizedAge
        B=&RandomRangeFloat(Minimum=0.5, Maximum=1, Randomness Mode=::NewEnumerator2, Random Seed=?)
       )
      ))
    )
   ) {isSetParameters=true}
   /SetVariables(
    Particles.Sprite_02_Color:LinearColor=&MultiplyLinearColor(
     Linear Color=$Particles.Initial.Sprite_02_Color
     Scale Factor=&MakeVector4_FromLinearColor(LinearColor=&MultiplyLinearColor(
       Linear Color=&FN_Select_Color_from_Time_Of_Day(
        Use Manual Time of Day Thresholds=true
        Morning=<rgba 1 0.75 0.7753 1>
        Day=<rgba 1 0.8842 0.8294 1>
        Evening=<rgba 1 0.625 0.5 1>
        Night=<rgba 0.7 0.3494 0.4483 1>
        Evening Threshold=21
       )
       Scale Factor=&MakeFloat4FromVectorAndFloat(
        "Vector (RGB)"=<1 1 1>
        "Float (Alpha)"=&Lerp_Float(
         A=1
         B=0
         Alpha=&Divide_Float(
          A=$Particles.NormalizedAge
          B=&RandomRangeFloat(Minimum=0.2, Maximum=1, Randomness Mode=::NewEnumerator2, Random Seed=?)
         )
        )
       )
      ))
    )
    Particles.Sprite_02_Size:Vector2f=&Multiply_Vector2D(
     A=$Particles.Initial.Sprite_02_Size
     B=&Vector2DFromFloat(Value=&FloatFromCurve(
       FloatCurve=DataInterface() {propertyValues {Curve {keys [<at 0.0018 0.1957>,<at 0.0527 0.819>,<at 0.1169 0.6555>,<at 0.1748 0.9015>,<at 0.2394 0.582>,<at 0.332 1.0023>,<at 0.4947 0.3707>,<at 0.6727 1>,<at 1 0>]}}}
       CurveIndex=&Divide_Float(
        A=$Particles.NormalizedAge
        B=&RandomRangeFloat(Minimum=0.5, Maximum=1, Randomness Mode=::NewEnumerator2, Random Seed=?)
       )
      ))
    )
   ) {isSetParameters=true}
   &KillParticles(Kill Particles=&SetBoolByFloatComparison(A=&VectorLength(Vector=$Particles.Velocity), B=1, Comparison Type=::NewEnumerator6))
  ]
  rendererClasses [&NiagaraSpriteRendererProperties]
  renderers [
   {rendererClass=&NiagaraSpriteRendererProperties}
   {rendererClass=&NiagaraSpriteRendererProperties, rendererIndex=1}
  ]
 }
 SmokeLWE {bEnabled=false, rendererClasses [], renderers []}
 SparklesLWE {rendererClasses [], renderers []}
}
summary {
 systemName=NS_Emote_SitPlant_Confetti
 emitters [
  {emitterName=SparklesLWE, rendererClasses []}
  {bEnabled=false, emitterName=AirglowLWE, rendererClasses []}
  {bEnabled=false, emitterName=SmokeLWE, rendererClasses []}
  {emitterName=Confetti_Medium, rendererClasses [&NiagaraSpriteRendererProperties]}
 ]
 userVariables [
  User.bIsFrontEnd {type=NiagaraBool}
  User.bIsFrontEndPreview {type=NiagaraBool}
  User.NormalizedNotifyProgress {type=NiagaraFloat, defaultValue=0, description=""}
 ]
}
